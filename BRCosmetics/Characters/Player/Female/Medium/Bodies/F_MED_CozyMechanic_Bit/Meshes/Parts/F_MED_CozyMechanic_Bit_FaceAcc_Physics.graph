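# Physics asset for the face accessory mesh of the female CozyMechanic_Bit character, covering a pigtail chain and the upper body around it.
# Mostly capsule bodies cover spine, neck, head, hair and both arms, with an extra large box on the left upper arm.
# Constraints lock swing and twist between the torso, neck, head and arm bodies.
# The four pigtail segments chain from based_pigtail, each joint with its own swing and twist limit angles.
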PhysicsAsset /BRCosmetics/Characters/Player/Female/Medium/Bodies/F_MED_CozyMechanic_Bit/Meshes/Parts/F_MED_CozyMechanic_Bit_FaceAcc_Physics
bodies {
 based_hair [{
   shapeType=Capsule
   length=17.3793
   radius=11.7749
   shapeName=based_hair_capsule
   center=<0.0001 0.0576 10.9093>
   rotation=<rot 89.6974 89.9337 179.9917>
  }]
 based_pigtail [{shapeType=Capsule, length=0, radius=0.5, shapeName=based_pigtail_capsule, rotation=<rot 0 -90 0>}]
 clavicle_l [{
   shapeType=Capsule
   length=6.6503
   radius=2.8818
   shapeName=clavicle_l_capsule
   center=<3.8601 -0.407 0.2582>
   rotation=<rot -85 179.9974 175.8258>
  }]
 clavicle_r [{shapeType=Capsule, length=0.505, radius=0.505, shapeName=clavicle_r_capsule, rotation=<rot 0 -90 0>}]
 hand_l [{shapeType=Capsule, length=0.505, radius=0.505, shapeName=hand_l_capsule, rotation=<rot 0 -90 0>}]
 hand_r [{shapeType=Capsule, length=0.505, radius=0.505, shapeName=hand_r_capsule, rotation=<rot 0 -90 0>}]
 head [{shapeType=Capsule, length=2.5317, radius=6.9437, center=<2.4463 0 0>, rotation=<rot 90 0 0>}]
 lowerarm_l [{shapeType=Capsule, length=14.1982, radius=4.6124, shapeName=lowerarm_l_capsule, center=<9.4079 0 0>, rotation=<rot -90 180 180>}]
 lowerarm_r [{shapeType=Capsule, length=14.1982, radius=4.6124, shapeName=lowerarm_l_capsule, center=<-9.4079 0 0>, rotation=<rot -90 0 180>}]
 neck_01 [{shapeType=Capsule, length=18.2597, radius=4.7895, shapeName=neck_01_capsule, center=<0.8644 -8.0323 0>, rotation=<rot 0 -20.0001 0>}]
 neck_02 [{shapeType=Capsule, length=0.1397, radius=3.9097, shapeName=neck_02_capsule}]
 pigtail_0 [{
   shapeType=Capsule
   length=0.7071
   radius=1.9302
   shapeName=pigtail_0_capsule
   center=<3.6235 0.8097 -0.0703>
   rotation=<rot -28.1427 -74.1773 -99.304>
  }]
 pigtail_1 [{shapeType=Capsule, length=2.1993, radius=2.2702, shapeName=pigtail_1_capsule, center=<3.4966 0.1094 0.1385>, rotation=<rot 0 87.88 -95.3403>}]
 pigtail_2 [{
   shapeType=Capsule
   length=6.4115
   radius=2.6772
   shapeName=pigtail_2_capsule
   center=<4.4123 0.1762 0.7173>
   rotation=<rot 0 -92.288 -89.7408>
  }]
 pigtail_3 [{
   shapeType=Capsule
   length=5.5407
   radius=2.108
   shapeName=pigtail_3_capsule
   center=<4.4257 -0.1158 0.1834>
   rotation=<rot 0 84.1819 -87.1566>
  }]
 spine_02 [{shapeType=Capsule, length=7.5351, radius=9.3848, shapeName=spine_02_capsule, center=<0.267 0 0>}]
 spine_03 [{shapeType=Capsule, length=7.0468, radius=8.5991, shapeName=spine_03_capsule, center=<0.6106 0.2666 0>}]
 spine_04 [{shapeType=Capsule, length=0.9117, radius=12.1012, shapeName=spine_04_capsule, center=<-1.3378 1.2629 0>}]
 spine_05 [
  {shapeType=Capsule, length=8.4434, radius=7.123, shapeName=spine_05_capsule, center=<0 0.5022 0>}
  {shapeType=Capsule, length=16.9184, radius=6.5729, shapeName=spine_05_capsule, center=<5.2433 -3.5121 0>}
 ]
 upperarm_l [
  {shapeType=Capsule, length=17.9478, radius=6.0958, shapeName=upperarm_l_capsule, center=<9.8981 0 2.133>, rotation=<rot -90 0 0>}
  {
   shapeType=Box
   extentX=39.0323
   extentY=7.5
   extentZ=26.9086
   shapeName=upperarm_l_box
   center=<12.9911 1.9231 16.0258>
   rotation=<rot 0 0 67.569>
  }
 ]
 upperarm_r [{shapeType=Capsule, length=16.9273, radius=6.9167, shapeName=upperarm_l_capsule, center=<-9.8981 0 0>, rotation=<rot -90 180 0>}]
}
constraints [
 {
  bone1Name=head
  bone2Name=neck_02
  swing1LimitDegrees=45
  swing1Motion=Locked
  swing2LimitDegrees=45
  swing2Motion=Locked
  twistLimitDegrees=45
  twistMotion=Locked
 }
 {
  bone1Name=spine_03
  bone2Name=spine_02
  swing1LimitDegrees=45
  swing1Motion=Locked
  swing2LimitDegrees=45
  swing2Motion=Locked
  twistLimitDegrees=45
  twistMotion=Locked
 }
 {
  bone1Name=spine_04
  bone2Name=spine_03
  swing1LimitDegrees=45
  swing1Motion=Locked
  swing2LimitDegrees=45
  swing2Motion=Locked
  twistLimitDegrees=45
  twistMotion=Locked
 }
 {
  bone1Name=spine_05
  bone2Name=spine_04
  swing1LimitDegrees=45
  swing1Motion=Locked
  swing2LimitDegrees=45
  swing2Motion=Locked
  twistLimitDegrees=45
  twistMotion=Locked
 }
 {
  bone1Name=clavicle_l
  bone2Name=spine_05
  swing1LimitDegrees=45
  swing1Motion=Locked
  swing2LimitDegrees=45
  swing2Motion=Locked
  twistLimitDegrees=45
  twistMotion=Locked
 }
 {
  bone1Name=upperarm_l
  bone2Name=clavicle_l
  swing1LimitDegrees=45
  swing1Motion=Locked
  swing2LimitDegrees=45
  swing2Motion=Locked
  twistLimitDegrees=45
  twistMotion=Locked
 }
 {
  bone1Name=lowerarm_l
  bone2Name=upperarm_l
  swing1LimitDegrees=45
  swing1Motion=Locked
  swing2LimitDegrees=45
  swing2Motion=Locked
  twistLimitDegrees=45
  twistMotion=Locked
 }
 {
  bone1Name=hand_l
  bone2Name=lowerarm_l
  swing1LimitDegrees=45
  swing1Motion=Locked
  swing2LimitDegrees=45
  swing2Motion=Locked
  twistLimitDegrees=45
  twistMotion=Locked
 }
 {
  bone1Name=clavicle_r
  bone2Name=spine_05
  swing1LimitDegrees=45
  swing1Motion=Locked
  swing2LimitDegrees=45
  swing2Motion=Locked
  twistLimitDegrees=45
  twistMotion=Locked
 }
 {
  bone1Name=upperarm_r
  bone2Name=clavicle_r
  swing1LimitDegrees=45
  swing1Motion=Locked
  swing2LimitDegrees=45
  swing2Motion=Locked
  twistLimitDegrees=45
  twistMotion=Locked
 }
 {
  bone1Name=lowerarm_r
  bone2Name=upperarm_r
  swing1LimitDegrees=45
  swing1Motion=Locked
  swing2LimitDegrees=45
  swing2Motion=Locked
  twistLimitDegrees=45
  twistMotion=Locked
 }
 {
  bone1Name=hand_r
  bone2Name=lowerarm_r
  swing1LimitDegrees=45
  swing1Motion=Locked
  swing2LimitDegrees=45
  swing2Motion=Locked
  twistLimitDegrees=45
  twistMotion=Locked
 }
 {
  bone1Name=neck_01
  bone2Name=spine_05
  swing1LimitDegrees=45
  swing1Motion=Locked
  swing2LimitDegrees=45
  swing2Motion=Locked
  twistLimitDegrees=45
  twistMotion=Locked
 }
 {
  bone1Name=neck_02
  bone2Name=neck_01
  swing1LimitDegrees=45
  swing1Motion=Locked
  swing2LimitDegrees=45
  swing2Motion=Locked
  twistLimitDegrees=45
  twistMotion=Locked
 }
 {
  bone1Name=upperarm_r
  bone2Name=spine_05
  swing1LimitDegrees=45
  swing1Motion=Locked
  swing2LimitDegrees=45
  swing2Motion=Locked
  twistLimitDegrees=45
  twistMotion=Locked
 }
 {bone1Name=pigtail_0, bone2Name=based_pigtail, swing1LimitDegrees=33, swing2LimitDegrees=33, twistLimitDegrees=8}
 {bone1Name=pigtail_1, bone2Name=pigtail_0, swing1LimitDegrees=33, swing2LimitDegrees=33, twistLimitDegrees=12.32}
 {bone1Name=pigtail_2, bone2Name=pigtail_1, swing1LimitDegrees=25, swing2LimitDegrees=38.824, twistLimitDegrees=8}
 {bone1Name=pigtail_3, bone2Name=pigtail_2, swing1LimitDegrees=25, swing2LimitDegrees=50.344, twistLimitDegrees=8}
]
