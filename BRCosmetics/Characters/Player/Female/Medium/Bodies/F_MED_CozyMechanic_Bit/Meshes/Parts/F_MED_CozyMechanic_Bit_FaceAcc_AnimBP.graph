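# Post-process style AnimBlueprint for the CozyMechanic Bit female cosmetic's face accessory mesh, adding secondary motion to hair and glasses.
# It copies the attached parent's pose, runs a Control Rig, then RigidBody simulation in spine_05 space and spring-limited AnimDynamics on left and right hair bones.
# Each update computes RigidBody gravity from a parameter asset and maps downward pawn velocity to an additive yaw on the glasses bone.
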AnimBlueprint /BRCosmetics/Characters/Player/Female/Medium/Bodies/F_MED_CozyMechanic_Bit/Meshes/Parts/F_MED_CozyMechanic_Bit_FaceAcc_AnimBP
target_skeleton=&/F_MED_CozyMechanic_Bit_FaceAcc_Skeleton
graphs [
 AnimGraph {nodes [
   n1 = Control Rig(Source=Copy Pose From Mesh() {bUseAttachedParent=true}) {controlRigAssetReference {blueprintRigClass=&/F_MED_CozyMechanic_Bit_FaceAcc_CtrlRig.F_MED_CozyMechanic_Bit_FaceAcc_CtrlRig_C}}
   n2 = RigidBody!(ComponentPose=Local To Component(LocalPose=@n1), Alpha, OverrideWorldGravity:Vector=Get Gravity Override(), ExternalForce) -> (Pose) {
    bClampLinearTranslationLimitToRefPose=true
    bForceDisableCollisionBetweenConstraintBodies=true
    bOverrideWorldGravity=true
    overridePhysicsAsset=&/F_MED_CozyMechanic_Bit_FaceAcc_Physics
    simulationSpace=BaseBoneSpace
    alphaBoolBlend {blendOption=Linear}
    baseBoneRef {boneName=spine_05}
    simSpaceSettings {
     dampingAlpha=0
     maxAngularAcceleration=80
     maxAngularVelocity=8
     maxLinearAcceleration=700
     maxLinearVelocity=30
     velocityScaleZ=0.7
     worldAlpha=1
     externalLinearDragV {x=0.5}
    }
   }
   n3 = Anim Dynamics - Bone: dyn_hair_r_0(ComponentPose=@n2, Alpha=1, ExternalForce="0,0,0") {
    angularDampingOverride=1
    angularSpringConstant=200
    bAngularSpring=true
    bOverrideAngularDamping=true
    bUseGravityOverride=true
    simulationSpace=RootRelative
    boundBone {boneName=dyn_hair_r_0}
    componentAppliedLinearAccClamp=<500 500 500>
    componentLinearAccScale=<0.7 0.7 0.7>
    componentLinearVelScale=<0.7 0.7 0.7>
    gravityOverride {z=-980}
    physicsBodyDefinitions [{
      boundBone {boneName=dyn_hair_r_0}
      boxExtents=<12 12 12>
      constraintSetup {angularLimitsMax=<12 7 4>, angularLimitsMin=<-20 2 -33>, angularTarget=<1 0 0>}
      localJointOffset=<1 0 0>
     }]
    relativeSpaceBone {boneName=head}
    simSpaceSettings {maxAngularAcceleration=35, maxAngularVelocity=7, simSpaceAngularAlpha=1}
   }
   n4 = Anim Dynamics - Bone: dyn_hair_r_2(ComponentPose=@n3, Alpha=1, ExternalForce="0,0,0") {
    angularDampingOverride=1
    angularSpringConstant=200
    bAngularSpring=true
    bOverrideAngularDamping=true
    bUseGravityOverride=true
    simulationSpace=RootRelative
    boundBone {boneName=dyn_hair_r_2}
    componentAppliedLinearAccClamp=<500 500 500>
    componentLinearAccScale=<0.7 0.7 0.7>
    componentLinearVelScale=<0.7 0.7 0.7>
    gravityOverride {z=-980}
    physicsBodyDefinitions [{
      boundBone {boneName=dyn_hair_r_2}
      boxExtents=<12 12 12>
      constraintSetup {angularLimitsMax=<12 22 5>, angularLimitsMin=<-12 -11 -15>, angularTarget=<1 0 0>}
      localJointOffset=<1 0 0>
     }]
    relativeSpaceBone {boneName=head}
    simSpaceSettings {maxAngularAcceleration=35, maxAngularVelocity=7, simSpaceAngularAlpha=1}
   }
   n5 = Anim Dynamics - Bone: dyn_hair_r_1(ComponentPose=@n4, Alpha=1, ExternalForce="0,0,0") {
    angularDampingOverride=1
    angularSpringConstant=200
    bAngularSpring=true
    bOverrideAngularDamping=true
    bUseGravityOverride=true
    simulationSpace=RootRelative
    boundBone {boneName=dyn_hair_r_1}
    componentAppliedLinearAccClamp=<500 500 500>
    componentLinearAccScale=<0.7 0.7 0.7>
    componentLinearVelScale=<0.7 0.7 0.7>
    gravityOverride {z=-980}
    physicsBodyDefinitions [{
      boundBone {boneName=dyn_hair_r_1}
      boxExtents=<12 12 12>
      constraintSetup {angularLimitsMax=<12 22 5>, angularLimitsMin=<-12 -11 -11>, angularTarget=<1 0 0>}
      localJointOffset=<1 0 0>
     }]
    relativeSpaceBone {boneName=head}
    simSpaceSettings {maxAngularAcceleration=35, maxAngularVelocity=7, simSpaceAngularAlpha=1}
   }
   n6 = Anim Dynamics - Bone: dyn_hair_l_0(ComponentPose=@n5, Alpha=1, ExternalForce="0,0,0") {
    angularDampingOverride=1
    angularSpringConstant=200
    bAngularSpring=true
    bOverrideAngularDamping=true
    bUseGravityOverride=true
    simulationSpace=RootRelative
    boundBone {boneName=dyn_hair_l_0}
    componentAppliedLinearAccClamp=<500 500 500>
    componentLinearAccScale=<0.7 0.7 0.7>
    componentLinearVelScale=<0.7 0.7 0.7>
    gravityOverride {z=-980}
    physicsBodyDefinitions [{
      boundBone {boneName=dyn_hair_l_0}
      boxExtents=<21 21 21>
      constraintSetup {angularLimitsMax=<18 6 33>, angularLimitsMin=<-20 1 -8>, angularTarget=<1 0 0>}
      localJointOffset=<1 0 0>
     }]
    relativeSpaceBone {boneName=head}
    simSpaceSettings {maxAngularAcceleration=35, maxAngularVelocity=7, simSpaceAngularAlpha=1}
   }
   n7 = Anim Dynamics - Bone: dyn_hair_l_2(ComponentPose=@n6, Alpha=1, ExternalForce="0,0,0") {
    angularDampingOverride=1
    angularSpringConstant=200
    bAngularSpring=true
    bOverrideAngularDamping=true
    bUseGravityOverride=true
    simulationSpace=RootRelative
    boundBone {boneName=dyn_hair_l_2}
    componentAppliedLinearAccClamp=<500 500 500>
    componentLinearAccScale=<0.7 0.7 0.7>
    componentLinearVelScale=<0.7 0.7 0.7>
    gravityOverride {z=-980}
    physicsBodyDefinitions [{
      boundBone {boneName=dyn_hair_l_2}
      boxExtents=<9 9 9>
      constraintSetup {angularLimitsMax=<18 5 15>, angularLimitsMin=<-12 -11 -15>, angularTarget=<1 0 0>}
      localJointOffset=<1 0 0>
     }]
    relativeSpaceBone {boneName=head}
    simSpaceSettings {maxAngularAcceleration=35, maxAngularVelocity=7, simSpaceAngularAlpha=1}
   }
   n8 = Anim Dynamics - Bone: dyn_hair_l_1(ComponentPose=@n7, Alpha=1, ExternalForce="0,0,0") {
    angularDampingOverride=1
    angularSpringConstant=200
    bAngularSpring=true
    bOverrideAngularDamping=true
    bUseGravityOverride=true
    simulationSpace=RootRelative
    boundBone {boneName=dyn_hair_l_1}
    componentAppliedLinearAccClamp=<500 500 500>
    componentLinearAccScale=<0.7 0.7 0.7>
    componentLinearVelScale=<0.7 0.7 0.7>
    gravityOverride {z=-980}
    physicsBodyDefinitions [{
      boundBone {boneName=dyn_hair_l_1}
      boxExtents=<9 9 9>
      constraintSetup {angularLimitsMax=<18 5 15>, angularLimitsMin=<-12 -11 -11>, angularTarget=<1 0 0>}
      localJointOffset=<1 0 0>
     }]
    relativeSpaceBone {boneName=head}
    simSpaceSettings {maxAngularAcceleration=35, maxAngularVelocity=7, simSpaceAngularAlpha=1}
   }
   n9 = "Transform (Modify) Bone - Bone: based_glasses"(ComponentPose=@n8, Alpha=1, Translation="0,0,0", Rotation=Get glass() -> (glass:Rotator), Scale="1,1,1") {rotationMode=BMM_Additive, rotationSpace=BCS_BoneSpace, translationMode=BMM_Additive, boneToModify {boneName=based_glasses}}
   Output Pose(Result=Component To Local(ComponentPose=@n9)) {layerGroup=None}
  ]}
 EventGraph {nodes [
   Event BlueprintUpdateAnimation()
   > n1 = "Gravity Override State Function (Parameter Asset) (Nativized)"(
    CharacterPartAnimInstance=Self-Reference()
    GameTimeInSeconds=GetGameTimeInSeconds()
    Parameters=/BRCosmetics/Characters/Player/Female/Medium/Bodies/F_MED_CozyMechanic_Bit/Meshes/M_MED_CozyMechanic_Bit_Parameters
   )
   > Set Gravity Override(Gravity Override=@n1)
   > Set!(
    execute
    glass_Roll:float=0
    glass_Pitch:float=0
    glass_Yaw:float=MapRangeClamped(Value=Get PawnVelocityZ(), InRangeA=0, InRangeB=-1000, OutRangeB=10)
   ) -> (then, Output_Get:Rotator)
  ]}
]
variables [
 body {type=SkeletalMeshComponent}
 body mesh {type=SkeletalMeshComponent}
 glass {type=Rotator}
 Gravity Override {type=Vector}
 NewVar {type=boolean}
]
